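# Unlit parent material for Nuhai wheel effects on skeletal meshes, drawing an animated equalizer ring with music notes.
# Time-stepped flipbooks of equalizer textures at two speeds are multiplied by channels of a fixed-frame ring texture.
# Custom and accent colors, varied by monochrome and hue-shift functions, tint the rings.
# A circle-masked, rotated notes texture cuts into the ring layer; the graph's three final branches show no material output connection.
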Material /VehicleCosmetics/Wheels/Nuhai/Materials/M_WheelParentEffect_Nuhai
graph {
 properties {
  bAutomaticallySetUsageInEditor=false
  bUsedWithMeshDeformer=true
  bUsedWithSkeletalMesh=true
  previewMesh=/VehicleCosmetics/Wheels/SR_14_T2/Meshes/SK_Wheel_SR_14_T2
  shadingModel=MSM_Unlit
  usedShadingModels=Unlit
 }
 nodes [
  n1 = Add(A=TextureCoordinate(), B=Constant() {r=-0.5})
  n2 = OneMinus(None=DotProduct(A=@n1, B=@n1))
  n3 = Constant() {r=0.97}
  n4 = Clamp(None=Multiply(A=Subtract(A=@n2, B=@n3), B=Constant() {r=30}))
  n5 = Add(A=TextureCoordinate() {uTiling=0.5, vTiling=0.5}, B=Constant2Vector() {g=-0.156, r=0.43})
  n6 = ScalarParameter() {defaultValue=3.5, parameterName=Scale}
  n7 = TextureSample(UVs=Rotator(Coordinate=Multiply(A=@n5, B=@n6), Time=Constant() {r=-0.1}) {speed=3.14}) {texture=/VehicleCosmetics/Wheels/NuhaiInverted/Textures/Boost_MusicNotes_N}
  n8 = Multiply(A=@n4, B=@n7.RGB)
  n9 = OneMinus(None=@n8)
  Add(A=Multiply(A=Clamp(None=@n9), B=Constant3Vector() {constant=<rgba 0.05 0.01 1 0>}), B=@n8)
  n10 = Multiply(A=@n4, B=@n7.A)
  n11 = VectorParameter() {parameterName=Custom_Color, defaultValue=<rgba 0.22 0.56 0.98 1>}
  n12 = ScalarParameter() {defaultValue=1.2, parameterName=Chromatic_Shift}
  n13 = Multiply(
   A=MaterialFunctionCall!(Color=@n11.RGB, Shift=@n12) {materialFunction=/VehicleCosmetics/Wheels/NuhaiInverted/Materials/MF_Monochrom_Shift}.Result
   B=Constant() {r=1.5}
  )
  n14 = TextureCoordinate()
  n15 = Constant() {r=1}
  n16 = ScalarParameter() {defaultValue=4, parameterName=Columns}
  n17 = AppendVector(A=Divide(A=@n15, B=@n16), B=Divide(A=@n15, B=@n16))
  n18 = Multiply(A=Time(), B=ScalarParameter() {defaultValue=8, parameterName=FB SPeed})
  n19 = ScalarParameter() {defaultValue=4, parameterName=Rows}
  n20 = Divide(A=Fmod(A=Floor(None=@n18), B=@n19), B=@n19)
  n21 = Divide(A=Floor(None=Divide(A=@n18, B=@n19)), B=@n16)
  n22 = TextureSample(UVs=Add(A=Multiply(A=@n14, B=@n17), B=AppendVector(A=@n20, B=@n21))) {samplerType=SAMPLERTYPE_LinearColor, texture=/VehicleCosmetics/Wheels/_SharedTextures/FX/EqualizerRing_RGB_FB_T}
  n23 = TextureCoordinate()
  n24 = Constant() {r=1}
  n25 = ScalarParameter() {defaultValue=4, parameterName=Columns}
  n26 = AppendVector(A=Divide(A=@n24, B=@n25), B=Divide(A=@n24, B=@n25))
  n27 = Constant() {r=1}
  n28 = ScalarParameter() {defaultValue=4, parameterName=Rows}
  n29 = Divide(A=Fmod(A=Floor(None=@n27), B=@n28), B=@n28)
  n30 = Divide(A=Floor(None=Divide(A=@n27, B=@n28)), B=@n25)
  n31 = TextureSample(UVs=Add(A=Multiply(A=@n23, B=@n26), B=AppendVector(A=@n29, B=@n30))) {samplerType=SAMPLERTYPE_LinearColor, texture=/VehicleCosmetics/Wheels/_SharedTextures/FX/EqualizerRing_RGB_FB_03_T}
  n32 = ComponentMask(None=@n31.RGB) {r=true}
  n33 = Multiply(A=@n32, B=@n22.G)
  n34 = ComponentMask(None=@n31.RGB) {g=true}
  n35 = Multiply(A=@n34, B=@n22.B)
  n36 = ComponentMask(None=@n31.RGB) {b=true}
  n37 = ScalarParameter() {defaultValue=2, parameterName=EQ_Ring_Brightness}
  n38 = Multiply(A=Add(A=Add(A=@n33, B=@n35), B=Multiply(A=@n36, B=@n22.G)), B=@n37)
  n39 = VectorParameter() {parameterName=Accent_Color, defaultValue=<rgba 0.98 0.04 0.96 1>}
  n40 = Add(A=Multiply(A=Multiply(A=@n13, B=@n22.G), B=@n38), B=Multiply(A=@n39.RGB, B=@n38))
  n41 = Multiply(
   A=MaterialFunctionCall(Color=@n39.RGB, Shift=@n12) {materialFunction=&MF_Monochrom_Shift}.Result
   B=Constant() {r=2}
  )
  n42 = TextureCoordinate()
  n43 = Constant() {r=1}
  n44 = ScalarParameter() {defaultValue=4, parameterName=Columns}
  n45 = AppendVector(A=Divide(A=@n43, B=@n44), B=Divide(A=@n43, B=@n44))
  n46 = Multiply(A=Time(), B=Constant() {r=3})
  n47 = ScalarParameter() {defaultValue=4, parameterName=Rows}
  n48 = Divide(A=Fmod(A=Floor(None=@n46), B=@n47), B=@n47)
  n49 = Divide(A=Floor(None=Divide(A=@n46, B=@n47)), B=@n44)
  n50 = TextureSample(UVs=Add(A=Multiply(A=@n42, B=@n45), B=AppendVector(A=@n48, B=@n49))) {samplerType=SAMPLERTYPE_LinearColor, texture=&EqualizerRing_RGB_FB_T}
  n51 = Multiply(A=@n32, B=@n22.R)
  n52 = Multiply(A=@n34, B=@n22.R)
  n53 = Multiply(A=@n37, B=Add(A=Multiply(A=@n36, B=@n50.R), B=Add(A=@n51, B=@n52)))
  n54 = MaterialFunctionCall!(Color=@n39.RGB, Angle=Constant() {r=0.7}) {materialFunction=/VehicleCosmetics/Wheels/Concept_Inverted/Materials/MF_HUE_Shift}
  n55 = Multiply(
   A=MaterialFunctionCall(Color=@n11.RGB, Angle=Constant() {r=0.2}) {materialFunction=&MF_HUE_Shift}.Result
   B=@n54.Result
  )
  n56 = Desaturation(None=@n31.RGB) {luminanceFactors=<rgba 0.3 0.59 0.11 0>}
  n57 = Multiply(A=@n56, B=Constant() {r=0.5})
  n58 = Add(A=Multiply(A=Multiply(A=@n41, B=@n22.G), B=@n53), B=Multiply(A=Multiply(A=@n55, B=@n54.Result), B=Add(A=@n53, B=@n57)))
  n59 = Multiply(A=@n22.A, B=Constant() {r=0.25})
  Multiply(A=Clamp(None=OneMinus(None=@n10)), B=Add(A=Add(A=@n40, B=@n58), B=Multiply(A=@n54.Result, B=@n59)))
  n60 = Clamp(None=Multiply(A=@n56, B=Constant() {r=100}))
  n61 = Multiply(A=@n55, B=@n50.R)
  n62 = Multiply(A=@n39.RGB, B=@n50.B)
  Multiply(A=Clamp(None=OneMinus(None=@n60)), B=Clamp(None=Add(A=@n61, B=@n62)))
 ]
}
